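# Niagara system for the jet exhaust trails of the MiddleSock glider, combining engine flame, forward sparks and speed lines.
# EngineFlames draws a mesh, sprite and light flame, Sparks_Forward emits GPU sprite sparks with glow, and SpeedLine001 spawns stretched streaks on a torus.
# System Timeline values ramp flame size, light brightness, spark rate and speed, and streak rate and opacity.
# Time-of-day and front-end modifiers adjust glow, brightness and sizing for gameplay versus menu preview.
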NiagaraSystem /BRCosmetics/Gadgets/Assets/VinderTech_GliderChute/Glider_MiddleSock/FX/NS_Glider_MiddleSock_JetTrails
dependencies {
 usedDataInterfaces [&NiagaraDataInterfaceCurve, &NiagaraDataInterfaceColorCurve, &NiagaraDataInterfaceVector2DCurve]
 usedDynamicInputs [&SetBoolByFloatComparison]
 usedModules [&SystemState, &Timeline]
}
emitters {
 EngineFlames {
  emitterUpdateScript [
   &EmitterState(
    Inactive Response=::NewEnumerator1
    Life Cycle Mode=::NewEnumerator1
    Loop Behavior=::NewEnumerator1
    Loop Duration Mode=::NewEnumerator1
    UseLoopDelay=false
   )
   &SpawnBurst_Instantaneous(Spawn Count=1, Use Spawn Probability=false)
  ]
  particleSpawnScript [
   &InitializeParticle(
    Color=<rgba 4 2.129 0.5147 1>
    Lifetime=0.25
    Mesh Scale=<0.6 0.6 0.5>
    Mesh Scale Mode=::NewEnumerator4
    Position Offset=<3 0 0>
    Sprite Size=<15 25>
    Sprite Size Mode=::NewEnumerator4
    UsePositionOffset=true
    Write Material Random=true
   )
   &InitialMeshOrientation(Orientation Axis=<-1 0 0>, Orientation Vector=<0 0 1>, Use Orientation Vector=true, Use Rotation Vector=false) {module=InitialMeshOrientation001}
   /SetVariables(Particles.CameraOffset_Outer=3) {isSetParameters=true}
   &SpriteFacingAndAlignment(Alignment Coordinate Space=::Local, Sprite Alignment=<1 0 0>, WriteAlignment=true, WriteFacing=false)
  ]
  particleUpdateScript [
   /SetVariables() {isSetParameters=true}
   &ParticleState(Kill Particles When Lifetime Has Elapsed=false)
   &ScaleSpriteSize(
    Scale Factor=&Lerp_Vector2(A=<1 1>, B=<1.1 2.5>, Alpha=$System.Timeline.NormalizedTime)
    Scale Sprite Size Mode=::NewEnumerator2
   )
   &ScaleColor(
    Scale Alpha=&FloatFromCurve(
     FloatCurve=DataInterface() {propertyValues {Curve {keys [<at 0 0 RCIM_Cubic>,<at 1 1 RCIM_Cubic>]}}}
     CurveIndex=$Emitter.Age
     Scale Curve=&Lerp_Float(
      A=&FNTimeOfDayScalar(Night Scale=0.01, Morning Scale=0.05, Day Scale=0.05, Evening Scale=0.01)
      B=&FNTimeOfDayScalar(Night Scale=0.05, Morning Scale=1, Evening Scale=0.05)
      Alpha=$System.Timeline.NormalizedTime
     )
    )
    ScaleA=true
    ScaleRGB=true
   )
   /SetVariables(Particles.Glow_Color:LinearColor=&MultiplyLinearColor(
     Linear Color=<rgba 1 0.265 0 1>
     Scale Factor=&MakeVector4(X=1, Y=1, Z=1, W=&FNTimeOfDayScalar(Night Scale=0.3, Morning Scale=1, Evening Scale=0.3))
    )) {isSetParameters=true}
   &ScaleMeshSize(Scale Factor=&MakeVector(
     X=&RandomRangeFloat(Minimum=0.1, Maximum=0.15, Evaluation Type=::NewEnumerator1)
     Y=&RandomRangeFloat(Minimum=0.1, Maximum=0.15, Evaluation Type=::NewEnumerator1)
     Z=&FloatFromCurve(
      FloatCurve=DataInterface() {propertyValues {Curve {keys [<at 0 0>,<at 0.85 1>]}}}
      Scale Curve=&Lerp_Float(A=0.4, B=1, Alpha=$System.Timeline.NormalizedTime)
     )
    ))
   &DynamicMaterialParameters(Module.Param0WriteEnabled=true, Module.Param1WriteEnabled=true, Module.Param2WriteEnabled=true, Module.Param3WriteEnabled=true) {enabled=false}
   &FN_Light_Attributes(
    Color=<rgba 1 0.1892 0.0393 1>
    Create Color Attribute=true
    Create Exponent Attribute=true
    Create Position Attribute=true
    Create Radius Attribute=true
    Exponent=5
    Position=<12 0 0>
    Radius=30
    Scale Color=&Lerp_Float(A=5, B=10, Alpha=$System.Timeline.NormalizedTime)
   )
  ]
  rendererClasses [
   &NiagaraMeshRendererProperties
   &NiagaraLightRendererProperties
   &NiagaraSpriteRendererProperties
  ]
  renderers [
   {rendererClass=&NiagaraMeshRendererProperties, rendererIndex=0}
   {rendererClass=&NiagaraMeshRendererProperties, rendererIndex=1}
   {rendererClass=&NiagaraLightRendererProperties, rendererIndex=2}
   {rendererClass=&NiagaraSpriteRendererProperties, rendererIndex=3}
  ]
 }
 Sparks_Forward {
  simTarget=GPUComputeSim
  emitterUpdateScript [
   &EmitterState(Delay First Loop Only=true, Inactive Response=::NewEnumerator1, Life Cycle Mode=::NewEnumerator1, Loop Duration=1, UseLoopDelay=true)
   &SpawnRate(SpawnRate=&Lerp_Float(A=5, B=15, Alpha=$System.Timeline.NormalizedTime), Use Spawn Probability=true)
  ]
  particleSpawnScript [
   &InitializeParticle(
    Color=<rgba 1.5 1.2105 0.384 1>
    Lifetime Max=0.6
    Lifetime Min=0.4
    Lifetime Mode=::NewEnumerator1
    Sprite Size Max=<1.25 4>
    Sprite Size Min=<1 2>
    Sprite Size Mode=::NewEnumerator2
    UsePositionOffset=false
    Write Material Random=true
   )
   &ShapeLocation(MaskSpawnGroup=false, Override Seed=false, Sphere Radius=2, UseOwnerScale=true)
   &AddVelocity(Velocity=<1 0 0>, Velocity Speed Scale=&Lerp_Float(A=125, B=250, Alpha=$System.Timeline.NormalizedTime)) {module=AddVelocity001}
   &CameraOffset(Camera Offset Amount=3)
  ]
  particleUpdateScript [
   &ParticleState()
   /SetVariables(Particles.Glow03_Color=&MakeLinearColorFromVectorAndFloat(
     "Vector (RGB)"=&MakeVectorFromLinearColorRGB(Color=&MultiplyLinearColor(
       Linear Color=&ColorFromCurve(ColorCurve=DataInterface() {propertyValues {
          AlphaCurve {keys []}
          BlueCurve {keys [<at 0 0.3733 RCIM_Cubic>,<at 0.5482 0.2623 RCIM_Cubic>]}
          GreenCurve {keys [<at 0 1.8979 RCIM_Cubic>,<at 0.5482 0 RCIM_Cubic>]}
          RedCurve {keys [<at 0 3 RCIM_Cubic>,<at 0.5482 1 RCIM_Cubic>]}
         }})
       Scale Factor=&MakeVector4_FromFloat(Float=$System.TOD)
      ))
     "Float (Alpha)"=&Multiply_Float(A=0.2, B=&FN_FrontEndFloatModifier(Gameplay=1, Front End=0.1))
    )) {isSetParameters=true}
   /SetVariables(Particles.Glow03_Size=&Multiply_Vector2D_ByFloat(
     Float=&FloatFromCurve(
      FloatCurve=DataInterface() {propertyValues {Curve {keys [
          <at 0 0 RCIM_Cubic>
          <at 0.2 1.3 RCIM_Cubic>
          {arriveTangent=-3.0745, interpMode=RCIM_Cubic, leaveTangent=-3.0745, tangentMode=RCTM_User, time=1, value=0}
         ]}}}
      Scale Curve=&FN_FrontEndFloatModifier(Gameplay=5, Front End=15)
     )
     Vector2D=<0.5 1.5>
    )) {isSetParameters=true}
   &ScaleColor(
    Scale Alpha=&FloatFromCurve(FloatCurve=DataInterface() {propertyValues {Curve {keys [<at 0.2 0>,<at 0.3 0.8>,<at 1 0>]}}})
    ScaleA=true
    ScaleRGB=true
   )
   &ScaleSpriteSize(Scale Factor=&Vector2DFromFloat(Value=&FloatFromCurve(
      FloatCurve=DataInterface() {propertyValues {Curve {keys [
          <at 0 0 RCIM_Cubic>
          {interpMode=RCIM_Cubic, leaveTangent=-3.0296, tangentMode=RCTM_Break, time=0.2, value=1.2}
          {arriveTangent=-1.8107, interpMode=RCIM_Cubic, leaveTangent=-1.8107, tangentMode=RCTM_User, time=1, value=0}
         ]}}}
      Scale Curve=&FN_FrontEndFloatModifier(Gameplay=0.5, Front End=0.3)
     ))) {module=ScaleSpriteSize001}
   &Drag(
    Drag=&Lerp_Float(A=4, B=1, Alpha=$System.Timeline.NormalizedTime)
    Use Linear Drag=true
    Use Rotational Drag=false
   )
   &SolveForcesAndVelocity(Clamp Velocity=false, Limit Acceleration=false)
  ]
  rendererClasses [&NiagaraSpriteRendererProperties]
  renderers [
   {rendererClass=&NiagaraSpriteRendererProperties}
   {rendererClass=&NiagaraSpriteRendererProperties, rendererIndex=1}
  ]
 }
 SpeedLine001 {
  emitterUpdateScript [
   &EmitterState()
   &SpawnRate(
    SpawnRate=&FN_FrontEndFloatModifier(Gameplay=&Lerp_Float(A=5, B=90, Alpha=$System.Timeline.NormalizedTime), Front End=50)
    Use Spawn Probability=true
   ) {module=SpawnRate001}
  ]
  particleSpawnScript [
   &InitializeParticle(
    Color=&MakeLinearColorFromVectorAndFloat(
     "Vector (RGB)"=<1 1 1>
     "Float (Alpha)"=&Multiply_Float(
      A=&UniformRangedFloat(Minimum=0.1, Maximum=0.18)
      B=&MakeCustomFloatFromBool(Bool=$User.bIsFrontEndPreview, True Float=0.5, False Float=1)
     )
    )
    Lifetime Max=0.16
    Lifetime Min=0.1
    Lifetime Mode=::NewEnumerator1
    Sprite Size Max=<3 10>
    Sprite Size Min=<2 2>
    Sprite Size Mode=::NewEnumerator2
    UsePositionOffset=true
    Write Material Random=true
   )
   &TorusLocation(
    Align Torus With Axis=true
    Handle Radius=0
    Large Radius:NiagaraFloat=&FN_FrontEndFloatModifier(
     Gameplay=&RandomRangeFloat(Minimum=100, Maximum=200)
     Front End=&RandomRangeFloat(Minimum=80, Maximum=120)
    )
    MaskSpawnGroup=false
    Module.Override Randomness=false
    Torus Axis=<1 0 0>
    Torus Origin:NiagaraPosition=&AddVectorToPosition(Position=$Particles.Position, Vector=<-300 100 0>)
   )
   &AddVelocityFromPoint(
    Module.Use Falloff Distance=false
    Origin Offset=&FN_FrontEndVector3Modifier(Gameplay=<-500 150 -5>, Front End=<-300 150 0>)
    Velocity Strength=600
   )
   &AddVelocity(Velocity=&Multiply_Vector(A=&RandomRangeVector(Minimum=<300 -150 0>, Maximum=<700 -150 0>), B=<1 1 1>))
  ]
  particleUpdateScript [
   &SolveForcesAndVelocity(Clamp Velocity=false, Limit Acceleration=false)
   &ParticleState()
   &ScaleSpriteSize(Scale Factor=&Vector2DFromCurve(
     Vector2Curve=DataInterface() {propertyValues {XCurve {keys [<at 0 0.5 RCIM_Cubic>,<at 0.5 1 RCIM_Cubic>]}, YCurve {keys [<at 0 0 RCIM_Cubic>,<at 0.5 1 RCIM_Cubic>]}}}
     ScaleCurve=&Vector2DFromFloat(Value=1)
    ))
   &ScaleSpriteSizeBySpeed(Max Scale Factor=<1 50>, Min Scale Factor=<0.01 0.05>, Velocity Threshold=1)
   &ScaleColor(
    Scale Alpha=&FN_FrontEndFloatModifier(
     Gameplay=&Lerp_Float(A=0, B=1, Alpha=$System.Timeline001.NormalizedTime)
     Front End=&Lerp_Float(A=0, B=1, Alpha=$System.Timeline001.NormalizedTime)
    )
    Scale RGB=&VectorFromFloat(Value=1)
    ScaleA=true
    ScaleRGB=true
   )
  ]
  rendererClasses [&NiagaraSpriteRendererProperties]
  renderers [{rendererClass=&NiagaraSpriteRendererProperties}]
 }
}
summary {
 systemName=NS_Glider_MiddleSock_JetTrails
 emitters [
  {
   emitterName=EngineFlames
   rendererClasses [
    &NiagaraMeshRendererProperties
    &NiagaraLightRendererProperties
    &NiagaraSpriteRendererProperties
   ]
  }
  {emitterName=Sparks_Forward, simTarget=GPUComputeSim, rendererClasses [&NiagaraSpriteRendererProperties]}
  {emitterName=SpeedLine001, rendererClasses [&NiagaraSpriteRendererProperties]}
 ]
 userVariables [
  User.bIsFrontEnd {type=NiagaraBool}
  User.bIsFrontEndPreview {type=NiagaraBool}
  User.bIsFullyDeployed {type=NiagaraBool}
  User.ForwardDot {type=NiagaraFloat, defaultValue=0, description=""}
 ]
}
